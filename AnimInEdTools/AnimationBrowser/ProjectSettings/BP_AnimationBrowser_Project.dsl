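; BP_AnimationBrowser_Project holds project settings for the Animation Browser tool, exposing file naming options for anim sets, descriptors, direction, subfolders and variations.
; Each naming category is an array of soft references to FileNameBase assets, which two shared helpers load with blocking loads and cast.
; GetFileNameBase merges their names into a de-duplicated list, while GetDataAssetBase returns the loaded assets themselves.
; GetProjectOptions returns the project options set as an array.
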(fn GetAnimSets
  (return (FileNaming|GetFileNameBase (Variables|FileNamingOptions|GetAnim%20Set))))

(fn GetAnimSetsDataAsset
  (return (DataAssets|GetDataAssetBase (Variables|FileNamingOptions|GetAnim%20Set))))

(fn GetDataAssetBase
  (params (Array :type Array%20of%20BP%20Animation%20Browser%20File%20Name%20Base%20Soft%20Object%20References))
  (Utilities|Array|Clear (LocalVariables|GetObjects%20Local))
  (for _array_element Array
    (bind _returnvalue (Utilities|LoadAssetBlocking _array_element))
    (bind _asbp_animation_browser_file_name_base (Utilities|Casting|CastToBP_AnimationBrowser_FileNameBase _returnvalue))
    (Utilities|Array|Add (LocalVariables|GetObjects%20Local) _asbp_animation_browser_file_name_base))
  (return (LocalVariables|GetObjects%20Local)))

(fn GetDescriptors
  (return (FileNaming|GetFileNameBase (Variables|FileNamingOptions|GetDescriptors))))

(fn GetDescriptorsDataAsset
  (return (DataAssets|GetDataAssetBase (Variables|FileNamingOptions|GetDescriptors))))

(fn GetDirection
  (return (FileNaming|GetFileNameBase (Variables|FileNamingOptions|GetDirection))))

(fn GetDirectionDataAsset
  (return (DataAssets|GetDataAssetBase (Variables|FileNamingOptions|GetDirection))))

(fn GetFileNameBase
  (params (Array :type Array%20of%20BP%20Animation%20Browser%20File%20Name%20Base%20Soft%20Object%20References))
  (Utilities|Array|Clear (LocalVariables|GetNames))
  (for _array_element Array
    (bind _returnvalue (Utilities|LoadAssetBlocking _array_element))
    (bind _asbp_animation_browser_file_name_base (Utilities|Casting|CastToBP_AnimationBrowser_FileNameBase _returnvalue))
    (for _array_element_2 (Variables|/AnimInEdTools/AnimationBrowser/ProjectSettings/Data/BP_AnimationBrowser_FileNameBase.BP_AnimationBrowser_FileNameBase_C|GetNames _asbp_animation_browser_file_name_base)
      (Utilities|Array|AddUnique (LocalVariables|GetNames) _array_element_2)))
  (return (LocalVariables|GetNames)))

(fn GetProjectOptions
  (bind _result (Utilities|Set|ToArray (Variables|ProjectOptions|GetProject%20Options)))
  (return _result))

(fn GetSubfolders
  (return (FileNaming|GetFileNameBase (Variables|FileNamingOptions|GetSubfolders))))

(fn GetVariations
  (return (FileNaming|GetFileNameBase (Variables|FileNamingOptions|GetVariations))))

(fn GetVariationsDataAsset
  (return (DataAssets|GetDataAssetBase (Variables|FileNamingOptions|GetVariations))))
