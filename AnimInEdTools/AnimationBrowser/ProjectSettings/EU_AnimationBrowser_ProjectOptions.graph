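# Editor utility base for Animation Browser project options, holding custom toolbar, context menu, naming-override and ShotGrid settings.
# Toolbar clicks store the button, and loading its menu clears the active browser's context menu, then loads one for that button's context.
# When a modal opens with name overrides enabled, it casts the widget to the animation dialog and replaces field names with the override strings.
# The project-selected hook is empty.
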EditorUtilityBlueprint /AnimInEdTools/AnimationBrowser/ProjectSettings/EU_AnimationBrowser_ProjectOptions
graphs [
 EventGraph {links [], nodes []}
 LoadToolbarContextMenu {nodes [
   LoadToolbarContextMenu() {is=title K2Node_FunctionEntry}
   n1 = Get Active Animation Browser() -> (Active Animation Browser:Animation Browser Object Reference)
   > Clear Context Menu(self:EUW Context Menu Object Reference=Get EUW_ContextMenu!(self:Animation Browser Object Reference=@n1) -> (EUW_ContextMenu:EUW Context Menu Object Reference))
   n2 = Get Clicked Toolbar Button() -> (Clicked Toolbar Button:EUW Anim Browser Toolbar Button Object Reference)
   > Load Dynamic Context Menu!(
    execute
    self:Animation Browser Object Reference=@n1
    Context:Name=Get Context!(self:EUW Anim Browser Toolbar Button Object Reference=@n2) -> (Context:Name=None)
    FocusButton:Widget Object Reference=@n2
   ) -> (then)
  ]}
 ModalDisplayed {nodes [
   n1 = ModalDisplayed!() -> (then, Modal:EUW Modal Dialog Object Reference, Widget:User Widget Object Reference) {is=title K2Node_FunctionEntry}
   > Branch(Condition=Get Override Modal Names() -> (Override Modal Names:bool))
   > n2 = Cast To EUW_Animation_Dialog!(execute, Object=@n1.Widget) -> (then, CastFailed, AsEUW Animation Dialog:EUW Animation Dialog Object Reference) {is=title K2Node_DynamicCast}
   > Update Field Names!(
    execute
    self:EUW Animation Dialog Object Reference=@n2
    Asset:String=Get Override Asset() -> (Override Asset:String)
    Plugin:String=Get Override Plugin() -> (Override Plugin:String)
    Subfolder:String=Get Override Subfolder() -> (Override Subfolder:String)
    AnimationSet:String=Get Override AnimationSet() -> (Override AnimationSet:String)
    Description:String=Get Override Description() -> (Override Description:String)
    Direction:String=Get Override Direction() -> (Override Direction:String)
    Mood/Status:String=Get Override Mood/Status() -> (Override Mood/Status:String)
   ) -> (then)
  ]}
 ProjectSelected {nodes [ProjectSelected() {is=title K2Node_FunctionEntry}]}
 ToolbarButtonClicked {nodes [
   n1 = ToolbarButtonClicked!() -> (then, Button:EUW Anim Browser Toolbar Button Object Reference) {is=title K2Node_FunctionEntry}
   > Set Clicked Toolbar Button(Clicked Toolbar Button:EUW Anim Browser Toolbar Button Object Reference=@n1) -> (Output_Get:EUW Anim Browser Toolbar Button Object Reference)
  ]}
]
variables [
 Active Animation Browser {type=Animation_Browser_C, default=null}
 Clicked Toolbar Button {type=EUW_AnimBrowserToolbarButton_C, default=null}
 Context Menu Options {type="array<S_AB_ContextMenuOptions>", default []}
 Custom Toolbar {type="array<S_AB_ToolbarOptions>", default []}
 Override AnimationSet {type=string, default=""}
 Override Asset {type=string, default=""}
 Override Description {type=string, default=""}
 Override Direction {type=string, default=""}
 Override Modal Names {type=boolean}
 Override Mood/Status {type=string, default=""}
 Override Plugin {type=string, default=""}
 Override Subfolder {type=string, default=""}
 ShotGrid Integration {type=boolean}
]
